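# Niagara system for the black-and-white stylized flame vehicle boost cosmetic, emitting from boost sockets on the skeletal mesh.
# A ribbon emitter with three renderers forms the trail, using per-group widths and material parameters, inherited source movement and camera-distance alpha.
# Animated flipbook flames add a scaled outline size, and two spawn groups of grayscale blobs spray in cones.
# Painted-variant overrides recolor ribbons and flames, while user parameters shift offsets for frontend preview and per-car placement.
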NiagaraSystem /VehicleCosmetics/Boosts/Boost_StylizedflameBW/FX/NS_Boost_StylizedflameBW
dependencies {
 usedDataInterfaces [
  &NiagaraDataInterfaceArrayFloat
  &NiagaraDataInterfaceCurve
  &NiagaraDataInterfaceCamera
  &NiagaraDataInterfaceColorCurve
  &NiagaraDataInterfaceArrayColor
  &NiagaraDataInterfaceSpriteRendererInfo
  &NiagaraDataInterfaceVector2DCurve
  &NiagaraDataInterfaceSkeletalMesh
 ]
 usedDynamicInputs []
 usedModules [
  &self:SystemSpawnScript.SystemScriptSource.SystemScriptGraph.NiagaraNodeAssignment_0.SetVariables_#1
  &SystemState
 ]
}
emitters {
 NE_VC_Boost_Ribbon {
  emitterUpdateScript [
   &EmitterState()
   &VC_FilteredSocketCount()
   &VC_RateQualityHandler(High Spawn Rate=60, Low Spawn Rate=30, Use Custom Rate?=true)
   &VC_SpawnRibbon_v1(Number of Ribbons=$Emitter.PortCount, SpawnRate=$Emitter.Spawn Rate)
  ]
  particleSpawnScript [
   &VC_PaintedVariantsOverride(DefaultColor=<rgba 0 0 0 1>, R1 DefaultColor=<rgba 1 1 1 9>, Use R1 Painted Overrides=true, Use R2 Painted Overrides=true)
   &VC_InitializeRibbon_v1(
    Color=$Particles.Color
    Custom Offset=true
    Enable Skeletal Position=true
    Lifetime=1.2
    Local Vector Offset=&AddVectorToPosition(
     Position=/Niagara/DynamicInputs/TypeConversions/MakeCustomPositionFromBool(Bool:NiagaraBool=$User.bIsFrontendPreview, True Position:NiagaraPosition=<-10 0 0>, False Position:NiagaraPosition=<60 0 0>)
     Vector=Linked() {linkedVariable {name=User.DTCarOffset, type=NiagaraPosition}}
    )
    Offset Amount=1
    Ribbon Width=100
   )
   &AddVelocity(Velocity=<-1000 0 0>) {module=AddVelocity001}
   /SetVariables(Particles.WidthRibbon3:NiagaraFloat=&SelectFloatFromArray(Direct Array Index=$Particles.SpawnGroup, Float Selection Array=DataInterface() {propertyValues {FloatData [-70,70]}})) {isSetParameters=true}
  ]
  particleUpdateScript [
   &ParticleState()
   &InheritSourceMovement(
    Applied Position Falloff Distance=5000
    Applied Position Scale=&Multiply_VectorByFloat(
     Vector=<1 1 1>
     Float=&FloatFromCurve(
      FloatCurve=DataInterface() {propertyValues {Curve {keys [
          {
           arriveTangentWeight=0.4252
           interpMode=RCIM_Cubic
           leaveTangentWeight=0.4252
           tangentMode=RCTM_User
           tangentWeightMode=RCTWM_WeightedBoth
           time=0.15
           value=1
          }
          {arriveTangentWeight=0.2833, interpMode=RCIM_Cubic, leaveTangentWeight=0.2833, tangentMode=RCTM_User, time=1, value=0}
         ]}}}
      Scale Curve=0.5
     )
    )
    ApplyForce=false
    ApplyPosition=true
    ApplyPositionFalloff=false
   )
   &SolveForcesAndVelocity(Clamp Velocity=false, Limit Acceleration=false)
   &VC_UpdateRibbon_v1(Enable Skeletal Position=true)
   &DynamicMaterialParameters(
    Index 0 Param 1=1
    Index 0 Param 2=1.25
    Index 0 Param 3=&SelectFloatFromArray(Direct Array Index=$Particles.SpawnGroup, Float Selection Array=DataInterface() {propertyValues {FloatData [0,0.5]}})
    Module.Index0Float0Write=true
    Module.Index0Float1Write=true
    Module.Index0Float2Write=true
    Module.Index0Float3Write=false
   )
   &ScaleColor(
    Scale Alpha=&MakeCustomFloatFromBool(
     Bool=$User.bIsFrontendPreview
     True Float=&Lerp_Float(
      A=5
      B=20
      Alpha=/Niagara/DynamicInputs/Camera/CameraFade(
       Camera Query:NiagaraDataInterfaceCamera=DataInterface()
       Fade Offset:NiagaraFloat=350
       Fade Distance:NiagaraFloat=500
       Use Owner Position:NiagaraBool=true
      )
     )
     False Float=&Lerp_Float(
      A=5
      B=20
      Alpha=&CameraFade(
       Camera Query:NiagaraDataInterfaceCamera=DataInterface()
       Fade Offset:NiagaraFloat=50
       Fade Distance:NiagaraFloat=150
       Use Owner Position:NiagaraBool=true
      )
     )
    )
    ScaleA=true
    ScaleRGB=true
   )
   &ScaleRibbonWidth(Ribbon Width Scale=&FloatFromCurve(
     FloatCurve=DataInterface() {propertyValues {Curve {keys [
         {arriveTangent=1.6117, interpMode=RCIM_Cubic, leaveTangent=1.6117, tangentMode=RCTM_User, time=0, value=0.6}
         {arriveTangent=-0.0567, interpMode=RCIM_Cubic, leaveTangent=-0.0567, tangentMode=RCTM_User, time=0.6, value=1.35}
         {arriveTangent=-5.6212, interpMode=RCIM_Cubic, leaveTangent=-5.6212, tangentMode=RCTM_User, time=1, value=-1}
        ]}}}
     Scale Curve=0.92
    ))
   /SetVariables(
    Particles.ColorScaleRibbon3:LinearColor=&MakeLinearColorFromVectorAndFloat(
     "Vector (RGB)"=&MakeVectorFromLinearColorRGB(Color=&ColorFromCurve(ColorCurve=DataInterface() {propertyValues {
         AlphaCurve {keys []}
         BlueCurve {keys [<at 0.391 0 RCIM_Cubic>,<at 0.4606 1 RCIM_Cubic>]}
         GreenCurve {keys [<at 0.391 0 RCIM_Cubic>,<at 0.4606 1 RCIM_Cubic>]}
         RedCurve {keys [<at 0.391 0 RCIM_Cubic>,<at 0.4606 1 RCIM_Cubic>]}
        }}))
     "Float (Alpha)"=&FloatFromCurve(
      FloatCurve=DataInterface() {propertyValues {Curve {keys [
          <at 0 1 RCIM_Cubic>
          {
           arriveTangent=0.1224
           arriveTangentWeight=0.1229
           interpMode=RCIM_Cubic
           leaveTangent=0.1224
           leaveTangentWeight=0.1229
           tangentMode=RCTM_User
           time=0.54
           value=0
          }
         ]}}}
      Scale Curve=10
     )
    )
    Particles.WidthRibbon2:NiagaraFloat=&FloatFromCurve(
     FloatCurve=DataInterface() {propertyValues {Curve {keys [<at 0 0 RCIM_Cubic>,<at 0.3 1.25 RCIM_Cubic>,<at 0.55 0 RCIM_Cubic>]}}}
     Scale Curve=&SelectFloatFromArray(Direct Array Index=$Emitter.SpawnGroup, Float Selection Array=DataInterface() {propertyValues {FloatData [-65,65]}})
    )
   ) {isSetParameters=true}
  ]
  rendererClasses [&NiagaraRibbonRendererProperties]
  renderers [
   {rendererClass=&NiagaraRibbonRendererProperties}
   {rendererClass=&NiagaraRibbonRendererProperties, rendererIndex=1}
   {rendererClass=&NiagaraRibbonRendererProperties, rendererIndex=2}
  ]
 }
 NE_VC_Boost_SpriteFlames {
  emitterUpdateScript [
   &EmitterState()
   &VC_FilteredSocketCount()
   &SpawnRate(SpawnRate=12, Use Spawn Probability=false)
  ]
  particleSpawnScript [
   &VC_PaintedVariantsOverride(
    Black=<rgba 0.19 0.19 0.19 8>
    BurntSienna=<rgba 0.65 0.2907 0.0779 8>
    Cobalt=<rgba 0.2844 0.4788 1.4 8>
    Crimson=<rgba 1.3 0.2505 0.2505 8>
    DefaultColor=<rgba 1 1 1 5>
    ForestGreen=<rgba 0.5604 1.1 0.3781 8>
    Grey=<rgba 0.4062 0.4062 0.4062 8>
    Lime=<rgba 0.8967 1 0.3802 8>
    Orange=<rgba 1.4 0.3872 0.1823 8>
    Pink=<rgba 1.4738 0.6094 1.5 8>
    Purple=<rgba 0.5558 0.1302 1 8>
    R1 Black:LinearColor=<rgba 0.05 0.05 0.05 2.6>
    R1 BurntSienna:LinearColor=<rgba 0.18 0.0764 0.0302 2.5>
    R1 Cobalt:LinearColor=<rgba 0.0673 0.1695 0.6458 3>
    R1 Crimson:LinearColor=<rgba 0.4219 0.0835 0.0835 2.5>
    R1 DefaultColor=&Multiply_LinearColorByFloat(Linear Color=<rgba 0 0 0 1>, Float=2)
    R1 ForestGreen:LinearColor=<rgba 0.1945 0.4427 0.1107 2.5>
    R1 Grey:LinearColor=<rgba 0.1094 0.1094 0.1094 2.5>
    R1 Lime:LinearColor=<rgba 0.3389 0.3854 0.1064 2.5>
    R1 Orange:LinearColor=<rgba 0.5208 0.1287 0.0651 2.5>
    R1 Pink:LinearColor=<rgba 0.526 0.1644 0.4959 2>
    R1 Purple:LinearColor=<rgba 0.1676 0.0072 0.3177 2>
    R1 Saffron:LinearColor=<rgba 0.5 0.4872 0.1432 2>
    R1 SkyBlue:LinearColor=<rgba 0.0752 0.514 0.6875 2>
    R1 TitaniumWhite:LinearColor=<rgba 0.474 0.474 0.474 3.5>
    Saffron=<rgba 1.15 1.117 0.2336 8>
    SkyBlue=<rgba 0.4167 1.2647 1.6 8>
    TitaniumWhite=<rgba 1.1 1.1 1.1 8>
    Use R1 Painted Overrides=true
   )
   &InitializeParticle(
    Color Mode=::NewEnumerator0
    Lifetime=0.55
    Sprite Rotation Mode=::NewEnumerator2
    Sprite Size Mode=::NewEnumerator3
    Uniform Sprite Size=61
    UsePositionOffset=false
    Write Material Random=true
   )
   &PartitionParticles(Number Of Partitions=$Emitter.PortCount, Particles Per Partition=1, Partition Mode:ENiagara_PartitionMode=::NewEnumerator1)
   &SkeletalMeshLocation(
    Bone / Socket Index=$Particles.PartitionParticles.Partition
    Bone / Socket Sampling Mode=::NewEnumerator5
    MaskSpawnGroup=false
    Mesh Sampling Type=::NewEnumerator6
    Sampled Position Offset=&AddVectorToPosition(
     Position=&MakeCustomPositionFromBool(Bool:NiagaraBool=$User.bIsFrontendPreview, True Position:NiagaraPosition=<-15 0 0>, False Position:NiagaraPosition=<-25 0 0>)
     Vector=Linked() {linkedVariable {name=User.DTCarOffset, type=NiagaraPosition}}
    )
    Skeletal Mesh=$System.SkeletalMesh
   )
   &AddVelocity(Velocity=<-150 0 0>)
  ]
  particleUpdateScript [
   &ParticleState()
   &CameraOffset(Camera Offset Amount=2)
   &SubUVAnimation(
    End Frame Range Override=24
    "Sprite Renderer "=DataInterface() {propertyValues {SpriteRenderer=&self:NE_VC_Boost_SpriteFlames_0.NiagaraSpriteRendererProperties_4}}
    SubUV Animation Mode:ENiagara_SubUVLookupModeV2=::NewEnumerator1
    SubUV Curve=DataInterface() {propertyValues {Curve {
       defaultValue=0
       keys [
        {arriveTangent=0.5133, interpMode=RCIM_Cubic, leaveTangent=0.5133, tangentMode=RCTM_User, time=0, value=0}
        {interpMode=RCIM_Cubic, tangentMode=RCTM_User, time=0.85, value=1}
       ]
      }}}
    UseEndFrame=false
    UseStartFrame=false
   )
   &SubUVAnimation(
    End Frame Range Override=24
    "Sprite Renderer "=DataInterface() {propertyValues {SpriteRenderer=&self:NE_VC_Boost_SpriteFlames_0.NiagaraSpriteRendererProperties_3}}
    SubUV Animation Mode:ENiagara_SubUVLookupModeV2=::NewEnumerator1
    SubUV Curve=DataInterface() {propertyValues {Curve {
       defaultValue=0
       keys [
        {arriveTangent=0.5133, interpMode=RCIM_Cubic, leaveTangent=0.5133, tangentMode=RCTM_User, time=0, value=0}
        {interpMode=RCIM_Cubic, tangentMode=RCTM_User, time=0.85, value=1}
       ]
      }}}
    UseEndFrame=false
    UseStartFrame=false
   ) {module=SubUVAnimation001}
   &SolveForcesAndVelocity(Clamp Velocity=false, Limit Acceleration=false)
   &ScaleSpriteSize(
    Scale Sprite Size Mode=::NewEnumerator0
    Uniform Scale Factor=&FloatFromCurve(FloatCurve=DataInterface() {propertyValues {Curve {keys [{tangentMode=RCTM_User, time=0, value=1}, {tangentMode=RCTM_User, time=1, value=2.5}]}}})
   )
   &ScaleColor(
    Scale Alpha=&FloatFromCurve(FloatCurve=DataInterface() {propertyValues {Curve {keys [<at 0 1>,<at 1 0>]}}})
    ScaleA=true
    ScaleRGB=true
   ) {module=ScaleColor001}
   /SetVariables(Particles.SpriteSizeOutline:Vector2f=&Multiply_Vector2D_ByFloat(Float=1.15, Vector2D=$Particles.SpriteSize)) {isSetParameters=true}
  ]
  rendererClasses [&NiagaraSpriteRendererProperties]
  renderers [
   {rendererClass=&NiagaraSpriteRendererProperties}
   {rendererClass=&NiagaraSpriteRendererProperties, rendererIndex=1}
  ]
 }
 NE_VC_Boost_Sprite_Blobs {
  emitterUpdateScript [
   &EmitterState()
   &VC_FilteredSocketCount()
   &SpawnRate(SpawnRate=50, Use Spawn Probability=false)
   &SpawnRate(SpawnGroup=1, SpawnRate=20, Use Spawn Probability=false) {module=SpawnRate001}
  ]
  particleSpawnScript [
   /SetVariables(Particles.VisibilityTag=$Emitter.SpawnGroup) {isSetParameters=true}
   &VC_PaintedVariantsOverride() {enabled=false}
   &InitializeParticle(
    Color=&SelectLinearColorFromArray(
     Array Sampling Mode=::NewEnumerator0
     Direct Array Index=?
     Color Selection Array=DataInterface() {propertyValues {ColorData [<rgba 0 0 0 1>, <rgba 0.81 0.81 0.81 1>, <rgba 0.5 0.5 0.5 1>]}}
     Random Seed=0
    )
    Lifetime Max=&MergeHelper_FloatSwitch(0=0.6, 1=0.6, 2=?, 3=?, Index=$Emitter.SpawnGroup, Num=::NewEnumerator0)
    Lifetime Min:NiagaraFloat=&MergeHelper_FloatSwitch(0=0.5, 1=0.4, 2=?, 3=?, Index=$Emitter.SpawnGroup, Num=::NewEnumerator0)
    Lifetime Mode=::NewEnumerator1
    Sprite Rotation Mode=::NewEnumerator2
    Sprite Size Mode=::NewEnumerator1
    Uniform Sprite Size Max=12
    Uniform Sprite Size Min=7
    UsePositionOffset=false
    Write Material Random=true
   )
   /SetVariables(Particles.HorizontalSpriteSize:Vector2f=&Multiply_Vector2D_ByFloat(Float=&RandomRangeFloat(Minimum=1, Maximum=1.7), Vector2D=$Particles.SpriteSize)) {isSetParameters=true}
   &PartitionParticles(Number Of Partitions=$Emitter.PortCount)
   &SkeletalMeshLocation(
    Bone / Socket Index=$Particles.PartitionParticles.Partition
    Bone / Socket Sampling Mode=::NewEnumerator5
    MaskSpawnGroup=false
    Mesh Sampling Type=::NewEnumerator6
    Sampled Position Offset=&AddVectorToPosition(Position=<-20 0 0>, Vector=Linked() {linkedVariable {name=User.DTCarOffset, type=NiagaraPosition}})
    Skeletal Mesh=$System.SkeletalMesh
   )
   &ShapeLocation(
    Cone Angle:NiagaraFloat=&MergeHelper_FloatSwitch(0=30, 1=0, 2=?, 3=?, Index=$Emitter.SpawnGroup, Num=::NewEnumerator0)
    Cone Inner Angle=40
    Cone Length:NiagaraFloat=&MergeHelper_FloatSwitch(0=30, 1=0, 2=?, 3=?, Index=$Emitter.SpawnGroup, Num=::NewEnumerator0)
    MaskSpawnGroup=false
    Override Seed=false
    Shape Primitive=::NewEnumerator5
    UseOwnerScale=true
   )
   &AddVelocity(
    Cone Angle=&MergeHelper_FloatSwitch(0=90, 1=35, 2=?, 3=?, Index=$Emitter.SpawnGroup, Num=::NewEnumerator0)
    Cone Axis=&MergeHelper_VectorSwitch(0=<0.4 0 -1>, 1=<1 0 0>, 2=?, 3=?, Index=$Emitter.SpawnGroup, Num=::NewEnumerator0)
    Distribution Along Cone Axis:NiagaraFloat=&MergeHelper_FloatSwitch(0=0.2, 1=1, 2=?, 3=?, Index=$Emitter.SpawnGroup, Num=::NewEnumerator0)
    Inner Cone Angle:NiagaraFloat=&MergeHelper_FloatSwitch(0=45, 1=15, 2=?, 3=?, Index=$Emitter.SpawnGroup, Num=::NewEnumerator0)
    Module.UseConeFalloff=true
    Module.UseVelDistribution=true
    Override Seed=false
    Speed Falloff From Cone Axis:NiagaraFloat=&MergeHelper_FloatSwitch(0=1, 1=0.5, 2=?, 3=?, Index=$Emitter.SpawnGroup, Num=::NewEnumerator0)
    Velocity Mode=::NewEnumerator2
    Velocity Speed=&MergeHelper_FloatSwitch(0=-310, 1=-500, 2=?, 3=?, Index=$Emitter.SpawnGroup, Num=::NewEnumerator0)
   )
  ]
  particleUpdateScript [
   &ParticleState()
   &ScaleSpriteSize()
   &Drag(Use Linear Drag=true, Use Rotational Drag=true)
   &SolveForcesAndVelocity(Clamp Velocity=false, Limit Acceleration=false)
   &ScaleColor(
    Scale Alpha=&FloatFromCurve(FloatCurve=DataInterface() {propertyValues {Curve {
        defaultValue=1
        keys [
         {tangentMode=RCTM_User, time=0, value=1}
         {arriveTangent=-0.0405, interpMode=RCIM_Cubic, leaveTangent=-0.0405, tangentMode=RCTM_User, time=0.81, value=1}
         {arriveTangent=-8.0089, interpMode=RCIM_Cubic, leaveTangent=-8.0089, tangentMode=RCTM_User, time=1, value=0}
        ]
       }}})
    ScaleA=true
    ScaleRGB=true
   )
  ]
  rendererClasses [&NiagaraSpriteRendererProperties]
  renderers [
   {rendererClass=&NiagaraSpriteRendererProperties}
   {rendererClass=&NiagaraSpriteRendererProperties, rendererIndex=1}
  ]
 }
}
summary {
 systemName=NS_Boost_StylizedflameBW
 emitters [
  {emitterName=NE_VC_Boost_Ribbon, rendererClasses [&NiagaraRibbonRendererProperties]}
  {emitterName=NE_VC_Boost_Sprite_Blobs, rendererClasses [&NiagaraSpriteRendererProperties]}
  {emitterName=NE_VC_Boost_SpriteFlames, rendererClasses [&NiagaraSpriteRendererProperties]}
 ]
 userVariables [
  User.bIsFrontendPreview {type=NiagaraBool}
  User.bIsPainted {type=NiagaraBool}
  User.ColorVariantID {type=NiagaraInt32}
  User.DTCarOffset {type=NiagaraPosition, description="", defaultValue {}}
  User.DTCarOffset {type=Vector3f, description="", defaultValue {struct=Vector3f, value=<0 0 0>}}
  User.DTCarOffset001 {type=Vector3f, description="", defaultValue {struct=Vector3f, value=<0 0 0>}}
  User.PaintedColor {type=LinearColor, description="", defaultValue {struct=LinearColor, value=<rgba 1 1 1 1>}}
  User.PaintedScalar {type=NiagaraFloat, defaultValue=1, description=""}
  User.PortCount {type=NiagaraInt32, defaultValue=2}
 ]
}
